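# Translucent, unlit, two-sided Niagara mesh-particle material, apparently for a transfer-effect cylinder on the InstantGravel glider.
# An upward or downward panning noise distorts a streak texture whose UVs are scaled and shifted by parameters.
# Three offset samples form RGB channels for a chromatic split, blending pale blue toward overbright white, tinted by particle color.
# Opacity combines the streak, V-axis masks, view-angle fade, particle alpha and depth fade.
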Material /BRCosmetics/Gadgets/Assets/VinderTech_GliderChute/Glider_InstantGravel/FX/M_Glider_InstantGravel_Transfer_Cylinder
graph {
 properties {
  bUsedWithNiagaraMeshParticles=true
  blendMode=BLEND_Translucent
  pixelDepthOffsetMode=PDOM_Legacy
  shadingModel=MSM_Unlit
  twoSided=true
  usedShadingModels=Unlit
 }
 nodes [
  n1 = ParticleColor()
  n2 = VectorParameter() {parameterName=UV Scale, channelNames {a=Noise V, b=Noise U, g=V Scale, r=U Scale}, defaultValue=<rgba 5 2 2 0.5>}
  n3 = Multiply(A=TextureCoordinate(), B=ComponentMask(None=@n2) {g=true, r=true})
  n4 = AppendVector(A=ComponentMask(None=@n3) {r=true}, B=ComponentMask(None=@n3) {g=true})
  n5 = AppendVector(A=Constant(), B=ScalarParameter() {parameterName=V Offset})
  n6 = TextureCoordinate()
  n7 = AppendVector(A=@n2.Noise U, B=@n2.Noise U)
  n8 = TextureSample(UVs=Panner(Coordinate=Multiply(A=@n6, B=@n7)) {speedY=-1}) {samplerType=SAMPLERTYPE_LinearGrayscale, texture=/Engine/Functions/Engine_MaterialFunctions02/ExampleContent/Textures/LowResBlurredNoise}
  n9 = ScalarParameter() {defaultValue=0.5, parameterName=Disturbance}
  n10 = TextureSampleParameter2D(UVs=Add(A=Add(A=@n4, B=@n5), B=Multiply(A=@n8.R, B=@n9) {constB=0.4})) {
   parameterName=Tex_Default
   samplerSource=SSM_Wrap_WorldGroupSettings
   texture=/Game/Accessories/FORT_Backpacks/Backpack_M_MED_OceanBreeze/FX/T_OceanBreeze_SplineB_t
  }
  n11 = Multiply(
   A=SphereMask(A=TextureCoordinate() {uTiling=0}, B=Constant2Vector() {g=0.5}) {attenuationRadius=0.4, hardnessPercent=0}
   B=SphereMask(A=TextureCoordinate() {uTiling=0}, B=Constant2Vector() {g=0.8}) {attenuationRadius=0.5, hardnessPercent=0}
  )
  n12 = TextureCoordinate()
  n13 = Power(Base=OneMinus(None=ComponentMask(None=@n12) {g=true})) {constExponent=3}
  n14 = Multiply(
   A=Power(Base=Multiply(A=@n10.R, B=@n11), Exp=ScalarParameter() {defaultValue=0.8, parameterName=Alpha power}) {constExponent=0.8}
   B=Saturate(None=Multiply(A=@n13) {constB=10})
  )
  n15 = MaterialFunctionCall("FadeSensitivity (optional)"=Constant() {r=1}) {materialFunction=&FadeBasedOnViewAngle}
  DepthFade(Opacity=Multiply(A=@n1.A, B=Multiply(A=@n14, B=@n15.Result))) {fadeDistanceDefault=5}
  n16 = Constant3Vector() {constant=<rgba 0.4062 0.8198 1 1>}
  n17 = Constant3Vector() {constant=<rgba 1.5 1.5 1.5 1>}
  n18 = ComponentMask(None=@n3) {g=true}
  n19 = Add(A=AppendVector(A=ComponentMask(None=@n3) {r=true}, B=Clamp(None=@n18)), B=@n5)
  n20 = Constant2Vector() {g=0.02, r=0.1}
  n21 = Multiply(A=@n8.R, B=@n9) {constB=0.5}
  n22 = TextureSampleParameter2D(UVs=Add(A=Subtract(A=@n19, B=@n20), B=@n21)) {
   parameterName=Tex_Default
   samplerSource=SSM_Wrap_WorldGroupSettings
   texture=&T_OceanBreeze_SplineB_t
  }
  n23 = TextureSampleParameter2D(UVs=Add(A=@n19, B=@n21)) {
   parameterName=Tex_Default
   samplerSource=SSM_Wrap_WorldGroupSettings
   texture=&T_OceanBreeze_SplineB_t
  }
  n24 = TextureSampleParameter2D(UVs=Add(A=Add(A=@n19, B=@n20), B=@n21)) {
   parameterName=Tex_Default
   samplerSource=SSM_Wrap_WorldGroupSettings
   texture=&T_OceanBreeze_SplineB_t
  }
  n25 = Multiply(
   A=MaterialFunctionCall(X=@n22.R, Y=@n23.R, Z=@n24.R) {materialFunction=&MakeFloat3}.Result
   B=ScalarParameter() {defaultValue=4, parameterName=Brightness}
  ) {constB=5}
  MaterialFunctionCall(Emissive=Multiply(A=@n1.RGB, B=LinearInterpolate(A=@n16, B=@n17, Alpha=@n25)) {constA=1}) {materialFunction=&MF_FX_EmissiveAdjustments}
  TextureSample(UVs=Panner(Coordinate=TextureCoordinate() {uTiling=2, vTiling=2}) {speedY=-1}) {samplerType=SAMPLERTYPE_LinearGrayscale, texture=&LowResBlurredNoise}
 ]
}
