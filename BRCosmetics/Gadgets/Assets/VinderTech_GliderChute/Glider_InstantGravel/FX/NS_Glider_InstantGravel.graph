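# Niagara effect for the InstantGravel glider cosmetic: deploy ribbon trails, a mesh trail, GPU sparks and speed lines.
# Forward and loop sparks crossfade spawn probability by User.ForwardDot, while a system timer ramps gameplay speed lines.
# The skeletal-mesh-attached Trail feeds smoothed values and lagged position into material parameters, tinted by time of day.
# Front-end modifiers retune timing, size and brightness; bIsFrontEndPreview picks which deploy ribbon spawns.
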NiagaraSystem /BRCosmetics/Gadgets/Assets/VinderTech_GliderChute/Glider_InstantGravel/FX/NS_Glider_InstantGravel
dependencies {
 usedDataInterfaces [&NiagaraDataInterfaceCurve, &NiagaraDataInterfaceColorCurve, &NiagaraDataInterfaceSkeletalMesh]
 usedDynamicInputs [&SetBoolByFloatComparison, &FNTimeOfDayScalar]
 usedModules [
  &self:SystemSpawnScript.SystemScriptSource.SystemScriptGraph.NiagaraNodeAssignment_0.SetVariables_#1
  &self:SystemSpawnScript.SystemScriptSource.SystemScriptGraph.NiagaraNodeAssignment_1.SetVariables_#2
  &self:SystemSpawnScript.SystemScriptSource.SystemScriptGraph.NiagaraNodeAssignment_2.SetVariables_#3
  &self:SystemSpawnScript.SystemScriptSource.SystemScriptGraph.NiagaraNodeAssignment_6.SetVariables_#4
  &self:SystemSpawnScript.SystemScriptSource.SystemScriptGraph.NiagaraNodeAssignment_3.SetVariables_#5
  &self:SystemSpawnScript.SystemScriptSource.SystemScriptGraph.NiagaraNodeAssignment_4.SetVariables_#6
  &self:SystemSpawnScript.SystemScriptSource.SystemScriptGraph.NiagaraNodeAssignment_5.SetVariables_#7
  &SystemState
  &FN_Timer
  &self:SystemSpawnScript.SystemScriptSource.SystemScriptGraph.NiagaraNodeAssignment_7.SetVariables_#8
 ]
}
emitters {
 Deploy_Trail {
  emitterUpdateScript [
   &EmitterState(
    Inactive Response=::NewEnumerator1
    Life Cycle Mode=::NewEnumerator1
    Loop Behavior=::NewEnumerator1
    Loop Delay=1
    Loop Duration=2
    UseLoopDelay=false
   )
   &FN_SpawnAnimTrail(Number of Anim Trails:NiagaraInt32=&MakeCustomIntFromBool(Bool=$User.bIsFrontEndPreview, True INT=0, False INT=1))
  ]
  particleSpawnScript [
   &FN_InitializeAnimTrailBoneIndecies()
   &FN_InitializeAnimTrail(AnimTrailWidth=1.2, Color=<rgba 0.3021 0.8684 2 0.2>, SkeletalMesh=$System.Skeletal Mesh AnimTrail)
   &DynamicMaterialParameters(
    Index 0 Param 1=1
    Index 0 Param 2=0.5
    Index 0 Param 3=1
    Index 0 Param 4=1
    Module.Param0WriteEnabled=true
    Module.Param1WriteEnabled=true
    Module.Param2WriteEnabled=true
    Module.Param3WriteEnabled=true
   )
   &AddVelocity(Rotation Coordinate Space=::Simulation, Velocity=<0 1000 0>, Velocity Speed Scale=1.5)
  ]
  particleUpdateScript [
   &ParticleState()
   &FN_UpdateAnimTrail(SkeletalMesh=$System.Skeletal Mesh AnimTrail)
   &ScaleRibbonWidth()
   &SolveForcesAndVelocity(Clamp Velocity=false, Limit Acceleration=false)
   &ScaleColor(
    Scale Alpha=&FloatFromCurve(
     FloatCurve=DataInterface() {propertyValues {Curve {keys [{interpMode=RCIM_Cubic, tangentMode=RCTM_User, time=0, value=1}, {interpMode=RCIM_Cubic, tangentMode=RCTM_User, time=1, value=0}]}}}
     Scale Curve=&FloatFromCurve(
      FloatCurve=DataInterface() {propertyValues {Curve {keys [
          <at 0 1 RCIM_Cubic>
          <at 0.5 1>
          {arriveTangent=-1.7149, interpMode=RCIM_Cubic, leaveTangent=-1.7149, tangentMode=RCTM_User, time=1, value=0}
         ]}}}
      CurveIndex=$Emitter.NormalizedLoopAge
      Scale Curve=&FN_FrontEndFloatModifier(
       Gameplay=&FloatFromCurve(FloatCurve=DataInterface() {propertyValues {Curve {keys [<at 0 1>,<at 0.6 1>,<at 0.9 0>]}}}, CurveIndex=$Emitter.Age)
       Front End=&FloatFromCurve(
        FloatCurve=DataInterface() {propertyValues {Curve {keys [<at 0.2 0>,<at 0.3 1>,<at 0.6 1>,<at 0.9 1>]}}}
        CurveIndex=$Emitter.Age
        Scale Curve=0
       )
      )
     )
    )
    ScaleA=true
    ScaleRGB=true
   )
  ]
  rendererClasses [&NiagaraRibbonRendererProperties]
  renderers [{rendererClass=&NiagaraRibbonRendererProperties}]
 }
 Deploy_Trail001 {
  emitterUpdateScript [
   &EmitterState(
    Inactive Response=::NewEnumerator1
    Life Cycle Mode=::NewEnumerator1
    Loop Behavior=::NewEnumerator1
    Loop Delay=1
    Loop Duration=2
    UseLoopDelay=false
   )
   &FN_SpawnAnimTrail(Number of Anim Trails:NiagaraInt32=&MakeCustomIntFromBool(Bool=$User.bIsFrontEndPreview, True INT=1, False INT=0))
  ]
  particleSpawnScript [
   &FN_InitializeAnimTrailBoneIndecies(First Bone or Socket Index=1, Second Bone or Socket Index=0)
   &FN_InitializeAnimTrail(AnimTrailWidth=1.2, Color=<rgba 0.3021 0.8684 2 0.2>, Lifetime=0.7, SkeletalMesh=$System.Skeletal Mesh AnimTrail)
   &DynamicMaterialParameters(
    Index 0 Param 1=1
    Index 0 Param 2=0.5
    Index 0 Param 3=1
    Index 0 Param 4=1
    Module.Param0WriteEnabled=true
    Module.Param1WriteEnabled=true
    Module.Param2WriteEnabled=true
    Module.Param3WriteEnabled=true
   )
   /SetVariables(Particles.Twis=0.25) {isSetParameters=true}
  ]
  particleUpdateScript [
   &ParticleState()
   &FN_UpdateAnimTrail(SkeletalMesh=$System.Skeletal Mesh AnimTrail)
   &ScaleRibbonWidth()
   &InheritSourceMovement(Applied Position Scale=<0.5 0.5 0.5>, ApplyForce=false, ApplyPosition=true, ApplyPositionFalloff=false)
   &SolveForcesAndVelocity(Clamp Velocity=false, Limit Acceleration=false)
   &ScaleColor(
    Scale Alpha=&FloatFromCurve(
     FloatCurve=DataInterface() {propertyValues {Curve {keys [{interpMode=RCIM_Cubic, tangentMode=RCTM_User, time=0, value=1}, {interpMode=RCIM_Cubic, tangentMode=RCTM_User, time=1, value=0}]}}}
     Scale Curve=&FloatFromCurve(
      FloatCurve=DataInterface() {propertyValues {Curve {keys [
          <at 0 1 RCIM_Cubic>
          <at 0.5 1>
          {arriveTangent=-1.7149, interpMode=RCIM_Cubic, leaveTangent=-1.7149, tangentMode=RCTM_User, time=1, value=0}
         ]}}}
      CurveIndex=$Emitter.NormalizedLoopAge
      Scale Curve=&FN_FrontEndFloatModifier(
       Gameplay=&FloatFromCurve(FloatCurve=DataInterface() {propertyValues {Curve {keys [<at 0 1>,<at 0.6 1>,<at 0.9 0>]}}}, CurveIndex=$Emitter.Age)
       Front End=&FloatFromCurve(FloatCurve=DataInterface() {propertyValues {Curve {keys [<at 0.1 0>,<at 0.6 1>,<at 0.9 1>]}}}, CurveIndex=$Emitter.Age)
      )
     )
    )
    ScaleA=true
    ScaleRGB=true
   )
  ]
  rendererClasses [&NiagaraRibbonRendererProperties]
  renderers [{rendererClass=&NiagaraRibbonRendererProperties}]
 }
 Sparks_Forward {
  simTarget=GPUComputeSim
  emitterUpdateScript [
   &EmitterState(
    Delay First Loop Only=true
    Life Cycle Mode=::NewEnumerator1
    Loop Delay=&FN_FrontEndFloatModifier(Gameplay=$System.Delay, Front End=1.5)
    Loop Duration=1
    UseLoopDelay=true
   )
   &SpawnRate(
    Spawn Probability=&Lerp_Float(
     A=0
     B=1
     Alpha=&ClampFloat(Float=&FN_FrontEndFloatModifier(Gameplay=$User.ForwardDot, Front End=1))
    )
    SpawnRate=100
    Use Spawn Probability=true
   )
  ]
  particleSpawnScript [
   &InitializeParticle(
    Color=<rgba 0.6083 1.3348 1.5 1>
    Lifetime Max=1
    Lifetime Min=0.4
    Lifetime Mode=::NewEnumerator1
    Sprite Size Max=<0.5 10>
    Sprite Size Min=<0.5 20>
    Sprite Size Mode=::NewEnumerator2
    UsePositionOffset=false
    Write Material Random=true
   )
   &AddVelocity(Velocity=<0 700 0>, Velocity Speed Scale=1.5) {module=AddVelocity001}
   &AddVelocity(Module.Use Falloff Distance=false, Override Seed=false, Velocity Mode=::NewEnumerator1, Velocity Speed=200)
  ]
  particleUpdateScript [
   &SkeletalMeshLocation(
    Bone Sampling Mode=::NewEnumerator4
    MaskSpawnGroup=false
    Position Sampling=::NewEnumerator1
    Sampled Position Offset=&AddVectorToPosition(Position=&RandomRangeVector(Minimum=<-20 100 0>, Maximum=<20 100 0>), Vector=$System.Offset)
    Skeletal Mesh=$System.Skeletal Mesh
   )
   &ParticleState()
   /SetVariables(Particles.Glow03_Color=&MakeLinearColorFromVectorAndFloat(
     "Vector (RGB)"=&MakeVectorFromLinearColorRGB(Color=&MultiplyLinearColor(
       Linear Color=&ColorFromCurve(ColorCurve=DataInterface() {propertyValues {
          AlphaCurve {keys []}
          BlueCurve {keys [<at 0 3 RCIM_Cubic>,<at 0.5482 1 RCIM_Cubic>]}
          GreenCurve {keys [<at 0 1.2884 RCIM_Cubic>,<at 0.5482 0.0104 RCIM_Cubic>]}
          RedCurve {keys [<at 0 0 RCIM_Cubic>,<at 0.5482 0.4952 RCIM_Cubic>]}
         }})
       Scale Factor=&MakeVector4_FromFloat(Float=1)
      ))
     "Float (Alpha)"=&Multiply_Float(A=0.2, B=&FN_FrontEndFloatModifier(Gameplay=1, Front End=0.5))
    )) {isSetParameters=true, module=SetVariables_#9}
   /SetVariables(Particles.Glow03_Size=&Multiply_Vector2D_ByFloat(
     Float=&FloatFromCurve(
      FloatCurve=DataInterface() {propertyValues {Curve {keys [
          <at 0 0 RCIM_Cubic>
          <at 0.2 1.3 RCIM_Cubic>
          {arriveTangent=-3.0745, interpMode=RCIM_Cubic, leaveTangent=-3.0745, tangentMode=RCTM_User, time=1, value=0}
         ]}}}
      Scale Curve=&FN_FrontEndFloatModifier(Gameplay=15, Front End=15)
     )
     Vector2D=<0.5 2>
    )) {isSetParameters=true, module=SetVariables_#10}
   &ScaleColor(ScaleA=true, ScaleRGB=true)
   &ScaleSpriteSize(Scale Factor=&Vector2DFromFloat(Value=&FloatFromCurve(
      FloatCurve=DataInterface() {propertyValues {Curve {keys [
          <at 0 0 RCIM_Cubic>
          {interpMode=RCIM_Cubic, leaveTangent=-3.0296, tangentMode=RCTM_Break, time=0.2, value=1.2}
          {arriveTangent=-1.8107, interpMode=RCIM_Cubic, leaveTangent=-1.8107, tangentMode=RCTM_User, time=1, value=0}
         ]}}}
      Scale Curve=&FN_FrontEndFloatModifier(Gameplay=0.5, Front End=0.3)
     ))) {module=ScaleSpriteSize001}
   &CurlNoiseForce(Noise Strength=&Multiply_Float(A=1000, B=1), Pan Noise=false) {module=CurlNoiseForce001}
   &Drag(Drag=3, Use Linear Drag=true, Use Rotational Drag=false)
   &SolveForcesAndVelocity(Clamp Velocity=false, Limit Acceleration=false)
   &ScaleSpriteSizeBySpeed(Max Scale Factor=<2 3>, Min Scale Factor=<0.5 0.1>)
  ]
  rendererClasses [&NiagaraSpriteRendererProperties]
  renderers [
   {rendererClass=&NiagaraSpriteRendererProperties}
   {rendererClass=&NiagaraSpriteRendererProperties, rendererIndex=1}
  ]
 }
 Sparks_Loop {
  simTarget=GPUComputeSim
  emitterUpdateScript [
   &EmitterState(Delay First Loop Only=true, Life Cycle Mode=::NewEnumerator1, Loop Delay=$System.Delay, Loop Duration=1, UseLoopDelay=true)
   &SpawnRate(
    Spawn Probability=&Lerp_Float(
     A=1
     B=0
     Alpha=&ClampFloat(Float=&FN_FrontEndFloatModifier(Gameplay=$User.ForwardDot, Front End=1))
    )
    SpawnRate=30
    Use Spawn Probability=true
   )
  ]
  particleSpawnScript [
   &InitializeParticle(
    Color=<rgba 0.6083 1.3348 1.5 1>
    Lifetime Max=1
    Lifetime Min=0.7
    Lifetime Mode=::NewEnumerator1
    Sprite Size Max=<4 15>
    Sprite Size Min=<3 10>
    Sprite Size Mode=::NewEnumerator2
    UsePositionOffset=false
    Write Material Random=true
   )
   &AddVelocity(Velocity=<0 300 0>) {module=AddVelocity001}
   &AddVelocity(Module.Use Falloff Distance=false, Override Seed=false, Velocity Mode=::NewEnumerator1, Velocity Speed=30)
  ]
  particleUpdateScript [
   &SkeletalMeshLocation(
    Bone Sampling Mode=::NewEnumerator4
    MaskSpawnGroup=false
    Position Sampling=::NewEnumerator1
    Sampled Position Offset=&AddVectorToPosition(Position=&RandomRangeVector(Minimum=<-20 100 0>, Maximum=<20 100 0>), Vector=$System.Offset)
    Skeletal Mesh=$System.Skeletal Mesh
   )
   &ParticleState()
   /SetVariables(Particles.Glow03_Color=&MakeLinearColorFromVectorAndFloat(
     "Vector (RGB)"=&MakeVectorFromLinearColorRGB(Color=&MultiplyLinearColor(
       Linear Color=&ColorFromCurve(ColorCurve=DataInterface() {propertyValues {
          AlphaCurve {keys []}
          BlueCurve {keys [<at 0 3 RCIM_Cubic>,<at 0.6334 1 RCIM_Cubic>]}
          GreenCurve {keys [<at 0 1.2884 RCIM_Cubic>,<at 0.6334 0.0104 RCIM_Cubic>]}
          RedCurve {keys [<at 0 0 RCIM_Cubic>,<at 0.6334 0.4952 RCIM_Cubic>]}
         }})
       Scale Factor=&MakeVector4_FromFloat(Float=1)
      ))
     "Float (Alpha)"=&Multiply_Float(A=0.2, B=1)
    )) {isSetParameters=true, module=SetVariables_#9}
   /SetVariables(Particles.Glow03_Size=&Multiply_Vector2D_ByFloat(
     Float=&FloatFromCurve(
      FloatCurve=DataInterface() {propertyValues {Curve {keys [
          <at 0 0 RCIM_Cubic>
          <at 0.2 1.3 RCIM_Cubic>
          {arriveTangent=-3.0745, interpMode=RCIM_Cubic, leaveTangent=-3.0745, tangentMode=RCTM_User, time=1, value=0}
         ]}}}
      Scale Curve=17
     )
     Vector2D=<1 3>
    )) {isSetParameters=true, module=SetVariables_#10}
   &ScaleColor(ScaleA=true, ScaleRGB=true)
   &ScaleSpriteSize(Scale Factor=&Vector2DFromFloat(Value=&FloatFromCurve(
      FloatCurve=DataInterface() {propertyValues {Curve {keys [
          <at 0 0 RCIM_Cubic>
          {interpMode=RCIM_Cubic, leaveTangent=-3.0296, tangentMode=RCTM_Break, time=0.2, value=1.2}
          {arriveTangent=-1.8107, interpMode=RCIM_Cubic, leaveTangent=-1.8107, tangentMode=RCTM_User, time=1, value=0}
         ]}}}
      Scale Curve=0.5
     ))) {module=ScaleSpriteSize001}
   &CurlNoiseForce(Noise Strength=&Multiply_Float(A=500, B=1), Pan Noise=false) {module=CurlNoiseForce001}
   &Drag(Use Linear Drag=true, Use Rotational Drag=false)
   &SolveForcesAndVelocity(Clamp Velocity=false, Limit Acceleration=false)
  ]
  rendererClasses [&NiagaraSpriteRendererProperties]
  renderers [
   {rendererClass=&NiagaraSpriteRendererProperties}
   {rendererClass=&NiagaraSpriteRendererProperties, rendererIndex=1}
  ]
 }
 SpeedLine {
  emitterUpdateScript [
   &EmitterState(
    Delay First Loop Only=true
    Inactive Response=::NewEnumerator1
    Life Cycle Mode=::NewEnumerator1
    Loop Delay=3
    Loop Duration=3
    UseLoopDelay=true
   )
   &SpawnRate(
    SpawnRate=&FN_FrontEndFloatModifier(
     Gameplay=&Multiply_Float(
      A=&Multiply_Float(A=200, B=1)
      B=&Lerp_Float(A=-0.8, B=1, Alpha=$System.FN_Timer.Value)
     )
     Front End=60
    )
    Use Spawn Probability=false
   )
  ]
  particleSpawnScript [
   &InitializeParticle(
    Lifetime Max=0.12
    Lifetime Min=0.1
    Lifetime Mode=::NewEnumerator1
    Position Offset=&FN_FrontEndVector3Modifier(Gameplay=<0 0 0>, Front End=<0 170 0>)
    Sprite Size Max=<2 7>
    Sprite Size Min=<2 2>
    Sprite Size Mode=::NewEnumerator2
    UsePositionOffset=true
    Write Material Random=true
   )
   &TorusLocation(
    Align Torus With Axis=true
    Handle Radius=0
    Large Radius:NiagaraFloat=&RandomRangeFloat(
     Minimum=&FN_FrontEndFloatModifier(Gameplay=250, Front End=100)
     Maximum=&FN_FrontEndFloatModifier(Gameplay=250, Front End=150)
    )
    MaskSpawnGroup=false
    Module.Override Randomness=false
    Offset=&FN_FrontEndVector3Modifier(Gameplay=<0 100 0>, Front End=<0 -100 0>)
    Torus Axis=<0 1 0>
   )
   &AddVelocity(Velocity=&RandomRangeVector(Minimum=<0 500 0>, Maximum=<0 500 0>))
   &AddVelocityFromPoint(
    Module.Use Falloff Distance=false
    Origin Offset=&FN_FrontEndVector3Modifier(Gameplay=<0 -200 0>, Front End=<0 -200 0>)
    Velocity Strength=&FN_FrontEndFloatModifier(Gameplay=1500, Front End=300)
   )
   &CameraOffset(Camera Offset Amount=-10)
  ]
  particleUpdateScript [
   &SolveForcesAndVelocity(Clamp Velocity=false, Limit Acceleration=false)
   &ParticleState()
   &ScaleSpriteSizeBySpeed(Max Scale Factor=<1 50>, Min Scale Factor=<0.01 0.05>, Velocity Threshold=1)
   &ScaleColor(
    Scale Alpha=&Multiply_Float(
     A=&FloatFromCurve(FloatCurve=DataInterface() {propertyValues {Curve {keys [
          {interpMode=RCIM_Cubic, tangentMode=RCTM_User, time=0, value=0}
          {interpMode=RCIM_Cubic, tangentMode=RCTM_User, time=0.5, value=1}
          {tangentMode=RCTM_User, time=1, value=0}
         ]}}})
     B=&Multiply_Float(
      A=&RandomRangeFloat(Minimum=0.1, Maximum=0.3)
      B=&FN_FrontEndFloatModifier(Gameplay=0.5, Front End=0.5)
     )
    )
    ScaleA=true
    ScaleRGB=true
   )
  ]
  rendererClasses [&NiagaraSpriteRendererProperties]
  renderers [{rendererClass=&NiagaraSpriteRendererProperties}]
 }
 Trail {
  emitterUpdateScript [
   &EmitterState(
    Inactive Response=::NewEnumerator1
    Life Cycle Mode=::NewEnumerator1
    Loop Behavior=::NewEnumerator1
    Loop Delay=$System.Delay
    Loop Duration Mode=::NewEnumerator1
    UseLoopDelay=true
   )
   &SpawnBurst_Instantaneous(Use Spawn Probability=false, UseLoopCountLimit=false)
  ]
  particleSpawnScript [
   &InitializeParticle(
    Color=<rgba 1.2 1.2 1.2 1.05>
    Mesh Scale=&Multiply_Vector(A=<1 1 1>, B=$System.Size Scale)
    Mesh Scale Mode=::NewEnumerator4
    UsePositionOffset=false
    Write Material Random=true
   )
   &InitialMeshOrientation(Mesh Orientation Mode=::NewEnumerator2, Rotation=<0 0 0.25>, Use Rotation Vector=true)
   &InitialMeshOrientation(Mesh Orientation Mode=::NewEnumerator2, Rotation=<0 0.25 0>, Use Rotation Vector=true) {module=InitialMeshOrientation001}
  ]
  particleUpdateScript [
   &DynamicMaterialParameters(
    Index 0 Param 1=&FN_FrontEndFloatModifier(Gameplay=$System.InterpolateOverTime.InterpolateOverTime_Float.Moving Average, Front End=1)
    Index 0 Param 2=&FN_FrontEndFloatModifier(
     Gameplay=&FloatFromCurve(
      FloatCurve=DataInterface() {propertyValues {Curve {keys [<at 1.9 0 RCIM_Cubic>, {arriveTangent=5.8989, interpMode=RCIM_Cubic, leaveTangent=5.8989, tangentMode=RCTM_User, time=2.55, value=1.4}]}}}
      CurveIndex=$Emitter.Age
     )
     Front End=&FloatFromCurve(
      FloatCurve=DataInterface() {propertyValues {Curve {keys [<at 1.3 0 RCIM_Cubic>, {arriveTangent=5.8989, interpMode=RCIM_Cubic, leaveTangent=5.8989, tangentMode=RCTM_User, time=1.95, value=1.4}]}}}
      CurveIndex=$Emitter.Age
     )
    )
    Index 0 Param 3=1
    Module.Index0Param0Write=true
    Module.Index0Param1Write=true
    Module.Index0Param2Write=true
    Module.Index0Param3Write=true
   )
   /SetVariables(Particles.DY2:Vector4f=&MakeVector4(
     X=0
     Y=&FN_FrontEndFloatModifier(
      Gameplay=&FloatFromCurve(
       FloatCurve=DataInterface() {propertyValues {Curve {keys [<at 1.9 0 RCIM_Cubic>, {arriveTangent=5.8989, interpMode=RCIM_Cubic, leaveTangent=5.8989, tangentMode=RCTM_User, time=2.55, value=1.4}]}}}
       CurveIndex=$Emitter.Age
      )
      Front End=&FloatFromCurve(
       FloatCurve=DataInterface() {propertyValues {Curve {keys [<at 1.3 0 RCIM_Cubic>, {arriveTangent=5.8989, interpMode=RCIM_Cubic, leaveTangent=5.8989, tangentMode=RCTM_User, time=1.95, value=1.4}]}}}
       CurveIndex=$Emitter.Age
      )
     )
     Z=0
     W=0
    )) {isSetParameters=true}
   &ParticleState(Kill Particles When Lifetime Has Elapsed=false, Let Infinitely Lived Particles Die When Emitter Deactivates=true)
   &ScaleMeshSize(Scale Factor=&MakeVector(
     X=&Lerp_Float(
      A=0.8
      B=1.1
      Alpha=&ClampFloat(Float=&FN_FrontEndFloatModifier(Gameplay=$System.InterpolateOverTime.InterpolateOverTime_Float.Moving Average, Front End=1))
     )
     Y=1
     Z=1
    ))
   &SkeletalMeshLocation(
    Bone Sampling Mode=::NewEnumerator4
    MaskSpawnGroup=false
    Sampled Position Offset=&AddVectorToPosition(Position=<0 80 0>, Vector=$System.Offset)
    Skeletal Mesh=$System.Skeletal Mesh
   )
   &FN_Mesh_Attributes(
    Color=&MultiplyLinearColor(
     Linear Color=&MakeCustomLinearColorFromBool(Bool=$User.bIsFrontEndPreview, True Color=<rgba 1 1 1 0.5>, False Color=<rgba 1 1 1 1.2>)
     Scale Factor=&MakeVector4_FromLinearColor(LinearColor=&FNTimeOfDayColor(Night Color=<rgba 1 0.333 0.4939 1>, Morning Color=<rgba 1 1 1 1>, Day Color=<rgba 1 1 1 1>, Evening Color=<rgba 1 1 1 1>))
    )
    Create Color Attribute=true
    Create Scale Attribute=true
    Scale=&Multiply_Vector(
     A=<1 0.3 1.3>
     B=&Multiply_Vector(
      A=$System.Size Scale
      B=&MakeVector(
       X=&Lerp_Float(
        A=0.8
        B=1.1
        Alpha=&ClampFloat(Float=&FN_FrontEndFloatModifier(Gameplay=$System.InterpolateOverTime.InterpolateOverTime_Float.Moving Average, Front End=1))
       )
       Y=1
       Z=1
      )
     )
    )
    Write to Module Namespace=true
   )
   &InterpolateOverTime(
    Rate Of Change:NiagaraFloat=$System.Interpolate Rate
    Target Vector Value:NiagaraPosition=$Particles.Position
    Variable Type=::NewEnumerator2
   )
   /SetVariables(Particles.Interpolate Dy:Vector4f=&MakeFloat4FromVectorAndFloat(
     "Vector (RGB)"=&Multiply_Vector(A=$Particles.InterpolateOverTime.InterpolateOverTime_V3_.Moving Average, B=<1 1 1>)
     "Float (Alpha)"=$System.Interpolate Internsity
    )) {isSetParameters=true}
   &ScaleColor(
    Scale RGB=&Multiply_VectorByFloat(
     Vector=&Multiply_VectorByFloat(Vector=<1 1 1>, Float=$System.TOD)
     Float=&Lerp_Float(
      A=1.2
      B=&FN_FrontEndFloatModifier(Gameplay=1, Front End=0.8)
      Alpha=&ClampFloat(Float=&FN_FrontEndFloatModifier(Gameplay=$System.InterpolateOverTime.InterpolateOverTime_Float.Moving Average, Front End=1))
     )
    )
    ScaleA=true
    ScaleRGB=true
   )
   &SolveForcesAndVelocity(Clamp Velocity=false, Limit Acceleration=false)
  ]
  rendererClasses [&NiagaraMeshRendererProperties]
  renderers [
   {rendererClass=&NiagaraMeshRendererProperties, rendererIndex=0}
   {rendererClass=&NiagaraMeshRendererProperties, rendererIndex=1}
  ]
 }
}
summary {
 systemName=NS_Glider_InstantGravel
 emitters [
  {emitterName=Trail, rendererClasses [&NiagaraMeshRendererProperties]}
  {emitterName=Sparks_Forward, simTarget=GPUComputeSim, rendererClasses [&NiagaraSpriteRendererProperties]}
  {emitterName=Sparks_Loop, simTarget=GPUComputeSim, rendererClasses [&NiagaraSpriteRendererProperties]}
  {emitterName=Deploy_Trail, rendererClasses [&NiagaraRibbonRendererProperties]}
  {emitterName=SpeedLine, rendererClasses [&NiagaraSpriteRendererProperties]}
  {emitterName=Deploy_Trail001, rendererClasses [&NiagaraRibbonRendererProperties]}
 ]
 userVariables [
  User.bIsFrontEnd {type=NiagaraBool}
  User.bIsFrontEndPreview {type=NiagaraBool}
  User.ForwardDot {type=NiagaraFloat, defaultValue=0, description=""}
  User.RibbonIDOffset {type=NiagaraInt32}
 ]
}
